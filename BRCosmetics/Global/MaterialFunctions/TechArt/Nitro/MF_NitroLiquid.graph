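# Material function that blends an animated, glowing nitro liquid with bubbles over input material attributes, using fill-level or custom masks.
# Custom HLSL ray-marches distorted 2D noise or a pseudo-volume texture for height and normals, with cheaper mobile and forced low-end paths.
# Height drives the color ramp, which shifts with a volatile state, beneath panning bubble layers.
# Niagara dynamic parameters can supply fill, surface plane, speed and turbulence.
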MaterialFunction /BRCosmetics/Global/MaterialFunctions/TechArt/Nitro/MF_NitroLiquid
graph {nodes [
  n1 = Time()
  n2 = Multiply(A=NamedRerouteUsage(), B=NamedRerouteUsage()) {constB=4}
  Add(A=@n1, B=@n2)
  n3 = NamedRerouteUsage()
  BumpOffset(
   Coordinate=@n3
   Height=Constant() {r=-1}
   HeightRatioInput=ScalarParameter() {defaultValue=0.2, group=Nitro_Liquid, parameterName=Bubble_HeightRatio_1}
  ) {heightRatio=0.1}
  n4 = NamedRerouteUsage()
  BumpOffset(
   Coordinate=@n4
   Height=Constant() {r=-1}
   HeightRatioInput=ScalarParameter() {defaultValue=0.3, group=Nitro_Liquid, parameterName=Bubble_HeightRatio_2}
  ) {heightRatio=0.1}
  Constant3Vector() {constant=<rgba 0 0 1 0>}
  Constant3Vector() {constant=<rgba 0.5 0.5 1 0>}
  Constant()
  Constant()
  n5 = ScalarParameter() {defaultValue=1.5, group=Nitro_Liquid, parameterName=NL_NitroPow}
  n6 = ScalarParameter() {defaultValue=1, group=Nitro_Liquid, parameterName=NL_NitroVolatilePow}
  n7 = ComponentMask(None=NamedRerouteUsage()) {r=true}
  n8 = Time()
  n9 = ScalarParameter() {defaultValue=1, parameterName=NL_01_Speed}
  n10 = MaterialFunctionCall(Float=Frac(None=Multiply(A=@n8, B=@n9))) {materialFunction=&MF_Hash}
  n11 = Multiply(
   A=LinearInterpolate(A=Power(Base=@n7), B=@n7, Alpha=NamedRerouteUsage())
   B=LinearInterpolate(A=Saturate(None=@n10.Float), Alpha=NamedRerouteUsage())
  )
  n12 = Multiply(A=@n11, B=LinearInterpolate(A=NamedRerouteUsage(), Alpha=NamedRerouteUsage()))
  n13 = Saturate(None=Power(Base=NamedRerouteUsage(), Exp=LinearInterpolate(A=@n5, B=@n6, Alpha=@n12) {constB=0}) {constExponent=4})
  CurveAtlasRowParameter(CurveTime=@n13) {
   atlas=/Game/Global/MaterialFunctions/TechArt/Nitro/CA_NitroColors
   curve=/Game/Global/MaterialFunctions/TechArt/Nitro/NitroLiquidColor
   parameterName=NitroLiquidColor
  }
  n14 = NamedRerouteUsage()
  n15 = NamedRerouteUsage()
  n16 = NamedRerouteUsage()
  n17 = NamedRerouteUsage()
  n18 = StaticSwitchParameter(True=Saturate(None=Add(A=@n16, B=@n17)), False=NamedRerouteUsage()) {parameterName=UseCustomInputMask}
  FunctionOutput(None=SetMaterialAttributes!(MaterialAttributes=BlendMaterialAttributes(A=@n14, B=@n15, Alpha=@n18) {vertexAttributeBlendType=UseA}))
  n19 = WorldPosition()
  n20 = TransformPosition(None=Constant3Vector()) {transformType=TRANSFORMPOSSOURCE_World}
  n21 = ObjectLocalBounds()
  n22 = MaterialFunctionCall() {materialFunction=&ObjectScale}
  n23 = Multiply(A=NamedRerouteDeclaration(None=Multiply(A=@n21.Extents, B=@n22.Scale XYZ)) {name=Obj_Dimensions, nodeColor=<rgba 0.4531 1 0 1>}) {constB=0.5}
  n24 = Divide(A=Add(A=Subtract(A=@n19.XYZ, B=@n20), B=@n23), B=@n21.Extents)
  n25 = Subtract(A=NamedRerouteUsage(), B=@n20)
  n26 = MaterialFunctionCall(Float3=Divide(A=Add(A=@n25, B=@n23), B=@n21.Extents) {constB=128}) {materialFunction=&BreakOutFloat3Components}
  n27 = MaterialFunctionCall(X=@n26.R, Y=@n26.R, Z=ComponentMask(None=@n24) {b=true}) {materialFunction=&MakeFloat3}
  n28 = NamedRerouteUsage()
  LinearInterpolate(A=@n24, B=@n27.Result, Alpha=@n28)
  MaterialFunctionCall!(Position, UVRadialScale) {materialFunction=/Game/Weapons/WeaponSkins/MaterialFunction/ObjectStabilizedScreenSpaceUVs}
  MaterialFunctionCall() {materialFunction=&CameraDirectionVector}
  n29 = NamedRerouteUsage()
  n30 = NamedRerouteUsage()
  n31 = Multiply(A=Multiply(A=Subtract(A=@n29, B=@n30)) {constB=0.5}, B=NamedRerouteUsage())
  n32 = MaterialFunctionCall(Float3=@n31) {materialFunction=&BreakOutFloat3Components}
  Max(A=Max(A=@n32.R, B=@n32.R), B=@n32.B)
  n33 = Multiply(A=@n4, B=ComponentMask(None=VectorParameter() {parameterName=BubbleTiling_2, defaultValue=<rgba 0.7 1.4 0 1>}.RGB) {g=true, r=true}) {constB=3}
  n34 = VectorParameter() {group=Nitro_Liquid, parameterName=NL_Bubble_PanSpeed, defaultValue=<rgba 0 0.25 0 1>}
  n35 = Add(A=NamedRerouteUsage(), B=Multiply(A=Time(), B=ComponentMask(None=@n34.RGB) {g=true, r=true}))
  n36 = Multiply(A=ComponentMask(None=VectorParameter() {group=Nitro_Liquid, parameterName=NL_Bubble_Tiling, defaultValue=<rgba 10 10 10 1>}.RGB) {g=true, r=true})
  n37 = Multiply(A=TextureSample(UVs=Multiply(A=@n35, B=@n36)) {texture=/Game/Characters/Player/Male/Medium/Bodies/M_MED_Viper/Textures/Glitch/T_SoftNoise}.RGB) {constB=2}
  n38 = Multiply(
   A=ComponentMask(None=Subtract(A=@n37)) {g=true, r=true}
   B=ScalarParameter() {defaultValue=0.01, group=Nitro_Liquid, parameterName=NL_Bubble_DistortionStrength}
  )
  n39 = Add(A=Add(A=Panner(Coordinate=@n33) {speedY=0.4}, B=NamedRerouteUsage()), B=@n38)
  n40 = TextureSample(UVs=@n39) {
   samplerSource=SSM_Wrap_WorldGroupSettings
   samplerType=SAMPLERTYPE_LinearGrayscale
   texture=/Game/Global/MaterialFunctions/TechArt/Nitro/T_NitroBubles
  }
  n41 = Multiply(A=@n3, B=ComponentMask(None=VectorParameter() {parameterName=BubbleTiling, defaultValue=<rgba 0.5 1 0 1>}.RGB) {g=true, r=true}) {constB=3}
  n42 = Add(A=Add(A=Panner(Coordinate=@n41) {speedY=0.4}, B=NamedRerouteUsage()), B=@n38)
  n43 = TextureSample(UVs=@n42) {samplerType=SAMPLERTYPE_LinearGrayscale, texture=&T_NitroBubles}
  n44 = LinearInterpolate(A=@n40.G, B=@n43.G, Alpha=@n43.G)
  n45 = Constant() {r=1}
  n46 = Multiply(A=NamedRerouteUsage()) {constB=2}
  n47 = TextureSample(UVs=Add(A=Panner(Coordinate=@n46) {speedY=0.1}, B=NamedRerouteUsage())) {texture=&T_SoftNoise}
  n48 = LinearInterpolate(Alpha=Saturate(None=Power(Base=@n47.R) {constExponent=1})) {constA=0.5}
  n49 = Constant()
  n50 = Saturate(None=Multiply(A=SphereMask(A=@n44, B=@n45, Radius=@n48, Hardness=@n49)) {constB=5})
  n51 = Multiply(A=NamedRerouteUsage()) {constB=2}
  Multiply(A=@n50, B=Saturate(None=@n51))
  n52 = VectorParameter() {group=Nitro_Liquid, parameterName=NL_EmisiveColor, defaultValue=<rgba 1 0.3866 0 1>}
  Multiply(A=@n52.RGB) {constB=3}
  n53 = LinearInterpolate(A=VectorParameter() {parameterName=NL_DarkColor, defaultValue=<rgba 0.0208 0.0019 0.0006 1>}.RGB, B=@n52.RGB, Alpha=@n13)
  n54 = LinearInterpolate(A=VectorParameter() {parameterName=NL_VolatileEmissiveColor, defaultValue=<rgba 0.3177 0.0283 0.0093 1>}.RGB, B=@n52.RGB, Alpha=@n13)
  n55 = Constant() {r=10}
  n56 = LinearInterpolate(
   A=Power(Base=Multiply(A=@n53) {constB=3}) {constExponent=1.5}
   B=Power(Base=Multiply(A=@n54, B=@n55)) {constExponent=3}
   Alpha=NamedRerouteUsage()
  ) {constB=0.5}
  Multiply(A=@n56, B=ScalarParameter() {defaultValue=15, parameterName=NL_Surface_Lightness})
  n57 = NamedRerouteUsage()
  n58 = OneMinus(None=@n57)
  n59 = LinearInterpolate(Alpha=NamedRerouteUsage()) {constA=14, constB=5}
  n60 = Max(A=@n57, B=Saturate(None=Power(Base=@n58, Exp=@n59) {constExponent=14}))
  Multiply(A=@n60) {constB=24}
  NamedRerouteDeclaration(None=FunctionInput(Preview=Constant() {r=1}) {bUsePreviewValueAsDefault=true, inputName=Mask, inputType=FunctionInput_Scalar, sortPriority=1}) {name=Mask, nodeColor=<rgba 0.4766 0 1 1>}
  n61 = Constant()
  n62 = SetMaterialAttributes!(MaterialAttributes, OpacityMask=@n61, Opacity=@n61) {attributeSetTypes [#1, #2]}
  NamedRerouteDeclaration(None=GetMaterialAttributes(None=FunctionInput(Preview=@n62) {bUsePreviewValueAsDefault=true, inputName=Material, inputType=FunctionInput_MaterialAttributes}).MaterialAttributes) {name=Input_Material, nodeColor=<rgba 1 0 0.0469 1>}
  NamedRerouteDeclaration(None=StaticSwitchParameter(True=NamedRerouteUsage(), False=NamedRerouteUsage()) {defaultValue=true, parameterName=UsePsuedovolumeTexture}) {name=RayMarchedNormal, nodeColor=<rgba 0.9688 0 1 1>}
  n63 = LinearInterpolate(
   A=TextureSample(UVs=@n39) {samplerType=SAMPLERTYPE_Normal, texture=/Game/Global/MaterialFunctions/TechArt/Nitro/T_NitroBubles_N}.RGB
   B=TextureSample(UVs=@n42) {samplerType=SAMPLERTYPE_Normal, texture=&T_NitroBubles_N}.RGB
   Alpha=@n43.G
  )
  NamedRerouteDeclaration(None=@n63) {name=BubbleNormal, nodeColor=<rgba 0 0.25 1 1>}
  NamedRerouteDeclaration(None=@n50) {name=BubbleMask, nodeColor=<rgba 0.4297 0 1 1>}
  n64 = NamedRerouteUsage()
  n65 = VectorParameter() {group=Nitro_Liquid, parameterName=NL_BubbleColor, defaultValue=<rgba 0.0208 0.0019 0.0006 1>}
  NamedRerouteDeclaration(None=Multiply(A=Multiply(A=@n64, B=@n65.RGB))) {name=BubbleColor, nodeColor=<rgba 0.4766 1 0 1>}
  NamedRerouteDeclaration(None=LinearInterpolate(A=@n56, B=@n56, Alpha=NamedRerouteUsage())) {name=LiquidColor, nodeColor=<rgba 0 1 0.6484 1>}
  n66 = NamedRerouteUsage()
  MaterialExpressionNamedRerouteUsage_16 = NamedRerouteUsage()
  MaterialExpressionTransform_1 = Transform(None=CameraVectorWS()) {transformSourceType=TRANSFORMSOURCE_World, transformType=TRANSFORM_Tangent}
  MaterialExpressionScalarParameter_11 = ScalarParameter() {defaultValue=16, group=NitroLiquid-2DRayMarch, parameterName=RayMarchSteps}
  MaterialExpressionScalarParameter_16 = ScalarParameter() {defaultValue=0.025, group=NitroLiquid-2DRayMarch, parameterName=RayMarchStepDist}
  MaterialExpressionNamedRerouteUsage_41 = NamedRerouteUsage()
  MaterialExpressionVertexNormalWS_1 = VertexNormalWS()
  MaterialExpressionNamedRerouteUsage_40 = NamedRerouteUsage()
  MaterialExpressionTime_4 = Time()
  MaterialExpressionNamedRerouteUsage_104 = NamedRerouteUsage()
  MaterialExpressionNamedRerouteUsage_106 = NamedRerouteUsage()
  MaterialExpressionNamedRerouteUsage_129 = NamedRerouteUsage()
  MaterialExpressionMultiply_46 = Multiply(A=NamedRerouteUsage(), B=NamedRerouteUsage()) {constB=4}
  n67 = Custom!(
   uv=@MaterialExpressionNamedRerouteUsage_16
   viewDir=@MaterialExpressionTransform_1
   steps=@MaterialExpressionScalarParameter_11
   stepdist=@MaterialExpressionScalarParameter_16
   tex=@MaterialExpressionNamedRerouteUsage_41
   vertnormal=@MaterialExpressionVertexNormalWS_1
   dTex=@MaterialExpressionNamedRerouteUsage_40
   Time=@MaterialExpressionTime_4
   NDPanSpeed=@MaterialExpressionNamedRerouteUsage_104
   NPanSpeed=@MaterialExpressionNamedRerouteUsage_106
   DistortionStrength=@MaterialExpressionNamedRerouteUsage_129
   LSpeed=@MaterialExpressionMultiply_46
  ) {
   code="float offset = 0.05;\r\nfloat3 rayStep = viewDir * -1;\r\nfloat2 uv2 = uv + Time*NDPanSpeed-float2(0.0,LSpeed);\r\nuv += Time*NPanSpeed-float2(0.0,LSpeed);\r\nfloat4 heightTex = Texture2DSample(tex,texSampler,uv);\r\nfloat4 distortionTex = Texture2DSample(dTex,texSampler,uv2);\r\n//float steps = 16;\r\n//float stepdist = float(0.1);\r\n\r\nfor(int i=0;i<=16;i++)\r\n{\r\nif(heightTex.r<=i/16.0)\r\n//if(heightTex.r<=0.1)\r\n{\r\nreturn float((heightTex.r*i)/16.0); \r\n//return float(i/16.0);\r\n}\r\nfloat stepct = heightTex.r * stepdist;\r\nuv += rayStep * stepct;\r\nuv2 += rayStep * stepct;\r\ndistortionTex = Texture2DSample(dTex,texSampler,uv2);\r\nfloat2 distortStr = distortionTex.xy *DistortionStrength;\r\nheightTex = Texture2DSample(tex,texSampler,uv.xy + distortStr);\r\n\r\nfloat offsetR = Texture2DSample(tex,texSampler,uv.xy + distortStr + (offset,0.0)).r;\r\nfloat offsetG = Texture2DSample(tex,texSampler,uv.xy + distortStr + (0.0,offset)).r;\r\nNr = heightTex.r - offsetR.r;\r\nNg = heightTex.r - offsetG.r;\r\nNb = 1.0;\r\n//RayMarchUV = uv.xy + distortStr;\r\n}\r\n\r\nreturn heightTex;\r\n"
   description=RayMarchTexture2
   outputType=CMOT_Float1
   showCode=true
   additionalOutputs [
    {outputName=RayMarchUV, outputType=CMOT_Float2}
    {outputName=Nr, outputType=CMOT_Float1}
    {outputName=Ng, outputType=CMOT_Float1}
    {outputName=Nb, outputType=CMOT_Float1}
   ]
   inputs [
    {inputName=uv, input {expression=&self:MaterialExpressionNamedRerouteUsage_16}}
    {inputName=viewDir, input {expression=&self:MaterialExpressionTransform_1}}
    {inputName=steps, input {expression=&self:MaterialExpressionScalarParameter_11}}
    {
     inputName=stepdist
     input {expression=&self:MaterialExpressionScalarParameter_16}
    }
    {inputName=tex, input {expression=&self:MaterialExpressionNamedRerouteUsage_41}}
    {
     inputName=vertnormal
     input {expression=&self:MaterialExpressionVertexNormalWS_1}
    }
    {inputName=dTex, input {expression=&self:MaterialExpressionNamedRerouteUsage_40}}
    {inputName=Time, input {expression=&self:MaterialExpressionTime_4}}
    {
     inputName=NDPanSpeed
     input {expression=&self:MaterialExpressionNamedRerouteUsage_104}
    }
    {
     inputName=NPanSpeed
     input {expression=&self:MaterialExpressionNamedRerouteUsage_106}
    }
    {
     inputName=DistortionStrength
     input {expression=&self:MaterialExpressionNamedRerouteUsage_129}
    }
    {inputName=LSpeed, input {expression=&self:MaterialExpressionMultiply_46}}
   ]
  }
  n68 = StaticSwitchParameter(True=LinearInterpolate(A=@n67.return, B=NamedRerouteUsage(), Alpha=NamedRerouteUsage()), False=@n67.return) {group=Nitro_Liquid, parameterName=UseBlending}
  NamedRerouteDeclaration(None=StaticSwitchParameter(
    True=@n66
    False=DataDrivenShaderPlatformInfoSwitch!(IsMobile=@n66, "!(IsMobile)"=@n68) {dDSPIPropertyNames [{inputName=IsMobile, propertyCondition=COND_True}]}
   ) {parameterName=ForceLowEndPath}) {name=2DHeightMap, nodeColor=<rgba 1 0 0.9375 1>}
  n69 = NamedRerouteUsage()
  MaterialExpressionNamedRerouteUsage_56 = NamedRerouteUsage()
  MaterialExpressionAppendVector_0 = AppendVector(A=NamedRerouteUsage(), B=NamedRerouteUsage())
  MaterialExpressionMultiply_4 = Multiply(A=NamedRerouteUsage(), B=NamedRerouteUsage())
  MaterialExpressionNamedRerouteUsage_70 = NamedRerouteUsage()
  MaterialExpressionNamedRerouteUsage_77 = NamedRerouteUsage()
  MaterialExpressionNamedRerouteUsage_73 = NamedRerouteUsage()
  MaterialExpressionNamedRerouteUsage_59 = NamedRerouteUsage()
  MaterialExpressionNamedRerouteUsage_60 = NamedRerouteUsage()
  n70 = Custom!(
   Tex=@MaterialExpressionNamedRerouteUsage_56
   XYFrames=@MaterialExpressionAppendVector_0
   NumFrames=@MaterialExpressionMultiply_4
   CurPos=@MaterialExpressionNamedRerouteUsage_70
   viewDir=@MaterialExpressionNamedRerouteUsage_77
   CurPosDistort=@MaterialExpressionNamedRerouteUsage_73
   stepsizebias=@MaterialExpressionNamedRerouteUsage_59
   distortstrength=@MaterialExpressionNamedRerouteUsage_60
  ) {
   code="// The below expression will get compiled\r\n// into the output of this node\r\nfloat accumdens = 0;\r\nfloat3 rayStep = viewDir * -1;\r\nfloat offset =0.025;\r\n//float cursample = 0;\r\nfloat3 cursample = PseudoVolumeTexture(Tex,TexSampler,CurPos,XYFrames,NumFrames);\r\nfloat3 cursampledistort = PseudoVolumeTexture(Tex,TexSampler,CurPosDistort,XYFrames,NumFrames);\r\nfor(int i=0;i<=16;i++)\r\n{\r\nif(cursample.r<=i/16.0)\r\n//if(heightTex.r<=0.1)\r\n{\r\nreturn float((cursample.r*i)/16.0); \r\n//return float(i/16.0);\r\n}\r\nfloat stepct = cursample.r * stepsizebias;\r\n//float stepct = stepsizebias;\r\nCurPos += rayStep * stepct;\r\nCurPosDistort += rayStep * stepct;\r\nfloat OffsetR = PseudoVolumeTexture(Tex,TexSampler,CurPos+float3(offset,0,0),XYFrames,NumFrames).r;\r\nfloat OffsetG = PseudoVolumeTexture(Tex,TexSampler,CurPos+float3(0,offset,0),XYFrames,NumFrames).r;\r\nfloat OffsetB = PseudoVolumeTexture(Tex,TexSampler,CurPos+float3(0,0,offset),XYFrames,NumFrames).r;\r\n\r\ncursampledistort = PseudoVolumeTexture(Tex,TexSampler,CurPosDistort,XYFrames,NumFrames)*distortstrength;\r\n\r\ncursample = PseudoVolumeTexture(Tex,TexSampler,CurPos+cursampledistort,XYFrames,NumFrames).r;\r\n\r\n//cursample *= (cursampledistort*1.0);\r\n\r\nNr = cursample - OffsetR;\r\nNg = cursample - OffsetG;\r\nNb = cursample - OffsetB;\r\n}\r\nreturn cursample;\r\n"
   description=ParallaxVolumeTexture
   showCode=true
   additionalOutputs [{outputName=Nr, outputType=CMOT_Float1}, {outputName=Ng, outputType=CMOT_Float1}, {outputName=Nb, outputType=CMOT_Float1}]
   inputs [
    {inputName=Tex, input {expression=&self:MaterialExpressionNamedRerouteUsage_56}}
    {inputName=XYFrames, input {expression=&self:MaterialExpressionAppendVector_0}}
    {inputName=NumFrames, input {expression=&self:MaterialExpressionMultiply_4}}
    {
     inputName=CurPos
     input {expression=&self:MaterialExpressionNamedRerouteUsage_70}
    }
    {
     inputName=viewDir
     input {expression=&self:MaterialExpressionNamedRerouteUsage_77}
    }
    {
     inputName=CurPosDistort
     input {expression=&self:MaterialExpressionNamedRerouteUsage_73}
    }
    {
     inputName=stepsizebias
     input {expression=&self:MaterialExpressionNamedRerouteUsage_59}
    }
    {
     inputName=distortstrength
     input {expression=&self:MaterialExpressionNamedRerouteUsage_60}
    }
   ]
  }
  n71 = Power(Base=Multiply(A=@n70.return) {constB=1.5})
  NamedRerouteDeclaration(None=StaticSwitchParameter(
    True=@n69
    False=DataDrivenShaderPlatformInfoSwitch(IsMobile=@n69, "!(IsMobile)"=@n71) {dDSPIPropertyNames [{inputName=IsMobile, propertyCondition=COND_True}]}
   ) {parameterName=ForceLowEndPath}) {name=VolumeHeightMap, nodeColor=<rgba 1 0 0.3047 1>}
  n72 = NamedRerouteUsage()
  n73 = ScalarParameter() {defaultValue=5, parameterName=Normal_Strength}
  n74 = Normalize(VectorInput=MaterialFunctionCall(X=Multiply(A=@n70.Nr, B=@n73), Y=Multiply(A=@n70.Ng, B=@n73), Z=Constant() {r=1}) {materialFunction=&MakeFloat3}.Result)
  NamedRerouteDeclaration(None=StaticSwitchParameter(
    True=@n72
    False=DataDrivenShaderPlatformInfoSwitch(IsMobile=@n72, "!(IsMobile)"=@n74) {dDSPIPropertyNames [{inputName=IsMobile, propertyCondition=COND_True}]}
   ) {parameterName=ForceLowEndPath}) {name=VolumeNormal, nodeColor=<rgba 1 0.4453 0 1>}
  n75 = DynamicParameter() {parameterIndex=2, defaultValue=<rgba 0 0 0 0>, paramNames [Speed,Turbulence,Param3,Param4]}
  NamedRerouteDeclaration(None=@n75.Speed) {name=Speed, nodeColor=<rgba 0 0.7891 1 1>}
  NamedRerouteDeclaration(None=Saturate(None=TwoSidedSign())) {name=Backface, nodeColor=<rgba 1 0 0.7031 1>}
  n76 = NamedRerouteUsage()
  n77 = TextureSample(UVs=@n67.RayMarchUV, Tex=NamedRerouteUsage()) {samplerType=SAMPLERTYPE_Normal, texture=/Game/Characters/Player/Male/Large/Bodies/M_LRG_Clash_V/Textures/Simple_Perlin_N}
  n78 = StaticSwitchParameter(True=LinearInterpolate(A=@n77.RGB, B=NamedRerouteUsage(), Alpha=NamedRerouteUsage()), False=@n77.RGB) {group=Nitro_Liquid, parameterName=UseBlending}
  NamedRerouteDeclaration(None=StaticSwitchParameter(
    True=@n76
    False=DataDrivenShaderPlatformInfoSwitch(IsMobile=@n76, "!(IsMobile)"=@n78) {dDSPIPropertyNames [{inputName=IsMobile, propertyCondition=COND_True}]}
   ) {parameterName=ForceLowEndPath}) {name=2DNormal, nodeColor=<rgba 1 0.4688 0 1>}
  n79 = DynamicParameter() {defaultValue=<rgba 0 0 1 0.5>, paramNames [Water Normal X, Water Normal Y, Water Normal Z, Fill]}
  NamedRerouteDeclaration(None=Normalize(VectorInput=MaterialFunctionCall(X=@n79.Water Normal X, Y=@n79.Water Normal X, Z=@n79.Water Normal X) {materialFunction=&MakeFloat3}.Result)) {name=WaterPlaneZ, nodeColor=<rgba 1 0 0.1875 1>}
  n80 = MaterialFunctionCall() {materialFunction=&BoundingBoxBased_0-1_UVW}
  NamedRerouteDeclaration(None=StaticSwitchParameter(True=@n80.RGB, False=LinearInterpolate(A=@n80.RGB, B=@n27.Result, Alpha=@n28)) {group=Nitro_Liquid, parameterName=LocalizedMask}) {name=BB_UVW, nodeColor=<rgba 0 0.8828 1 1>}
  NamedRerouteDeclaration(None=@n75.Turbulence) {name=Turbulence, nodeColor=<rgba 0 0.7891 1 1>}
  NamedRerouteDeclaration(None=StaticSwitchParameter(True=@n79.Fill, False=ScalarParameter() {defaultValue=1, group=Static Overrides, parameterName=Liquid Fill}) {defaultValue=true, parameterName=Is Niagara}) {name=Fill, nodeColor=<rgba 0 1 0.0859 1>}
  n81 = PreSkinnedLocalBounds()
  n82 = MaterialFunctionCall() {materialFunction=&ObjectScale}
  NamedRerouteDeclaration(None=Multiply(A=@n81.Extents, B=@n82.Scale XYZ)) {name=Dimensions, nodeColor=<rgba 0 0.0994 1 1>}
  n83 = DynamicParameter() {parameterIndex=1, defaultValue=<rgba 1 0 0 0>, paramNames [Tangent X, Tangent Y, Tangent Z, Time]}
  NamedRerouteDeclaration(None=MaterialFunctionCall(X=@n83.Tangent X, Y=@n83.Tangent X, Z=@n83.Tangent X) {materialFunction=&MakeFloat3}.Result) {name=WaterPlaneTangent, nodeColor=<rgba 0.8984 1 0 1>}
  n84 = MaterialFunctionCall!(Turbulance=NamedRerouteUsage()) {materialFunction=/Game/Global/MaterialFunctions/TechArt/Nitro/MF_FluidMasks}
  NamedRerouteDeclaration(None=@n84.FluidSurfaceWP) {name=FluidSurfaceUV, nodeColor=<rgba 1 0 0.2812 1>}
  n85 = DotProduct(A=Constant3Vector() {constant=<rgba 0 0 -1 0>}, B=NamedRerouteUsage())
  NamedRerouteDeclaration(None=Multiply(A=Add(A=@n85)) {constB=0.5}) {name=WaterUpsideDown, nodeColor=<rgba 0 0.1797 1 1>}
  NamedRerouteDeclaration(None=OneMinus(None=@n84.TwosidedMask)) {name=TwoSidedMask, nodeColor=<rgba 0 1 0.3906 1>}
  n86 = Power(Base=Multiply(A=NamedRerouteUsage()) {constB=4})
  n87 = NamedRerouteUsage()
  NamedRerouteDeclaration(None=Multiply(A=@n11, B=LinearInterpolate(A=@n86, Alpha=@n87))) {name=Lightning/VolatileMask, nodeColor=<rgba 1 0.1172 0 1>}
  n88 = TextureObject() {
   samplerType=SAMPLERTYPE_LinearGrayscale
   texture=/Game/Global/Materials/MaterialGym/Materials/SmokeWithFlowmap/Textures/T_Swirly_Noise_MiddleDetail
  }
  NamedRerouteDeclaration(None=@n88) {name=NoiseTexture, nodeColor=<rgba 0.2891 1 0 1>}
  NamedRerouteDeclaration(None=ScalarParameter() {defaultValue=0.1, group=NitroLiquid-2DRayMarch, parameterName=DistortionStrength}) {name=DistortionStrength, nodeColor=<rgba 0.4766 0 1 1>}
  NamedRerouteDeclaration(None=ComponentMask(None=VectorParameter() {group=NitroLiquid-2DRayMarch, parameterName=NL_NoiseTexturePanSpeed, defaultValue=<rgba 0 -0.1 0 1>}.RGB) {g=true, r=true}) {name=NoiseTexturePanSpeed, nodeColor=<rgba 0.9219 0 1 1>}
  n89 = NamedRerouteUsage()
  n90 = Multiply(A=Time(), B=NamedRerouteUsage())
  n91 = NamedRerouteUsage()
  n92 = Multiply(A=Time(), B=NamedRerouteUsage())
  n93 = TextureSample(
   UVs=BumpOffset(Coordinate=Add(A=@n91, B=@n92), Height=Constant() {r=0.5}, HeightRatioInput=NamedRerouteUsage()) {heightRatio=-0.25, referencePlane=0}
   Tex=NamedRerouteUsage()
  )
  n94 = ComponentMask(None=@n93.RGB) {g=true, r=true}
  n95 = Multiply(A=NamedRerouteUsage()) {constB=0.5}
  n96 = BumpOffset(
   Coordinate=Add(A=Add(A=@n89, B=@n90), B=Multiply(A=@n94, B=@n95) {constB=0})
   Height=LinearInterpolate(Alpha=ComponentMask(None=@n93.RGB) {r=true}) {constB=0.5}
   HeightRatioInput=NamedRerouteUsage()
  ) {heightRatio=-0.25, referencePlane=0}
  n97 = NamedRerouteUsage()
  n98 = Saturate(None=Power(Base=TextureSample(UVs=@n96, Tex=@n97).R) {constExponent=2.5})
  n99 = NamedRerouteUsage()
  n100 = NamedRerouteUsage()
  NamedRerouteDeclaration(None=StaticSwitchParameter(True=LinearInterpolate(A=@n98, B=@n99, Alpha=@n100), False=@n98) {group=Nitro_Liquid, parameterName=UseBlending}) {name=LowEnd2DHeightTexture, nodeColor=<rgba 0.6172 0 1 1>}
  NamedRerouteDeclaration(None=TextureObject() {samplerType=SAMPLERTYPE_Normal, texture=&Simple_Perlin_N}) {name=NormalTexture, nodeColor=<rgba 0.2891 1 0 1>}
  n101 = TextureSample(UVs=@n96, Tex=NamedRerouteUsage())
  n102 = NamedRerouteUsage()
  n103 = NamedRerouteUsage()
  NamedRerouteDeclaration(None=StaticSwitchParameter(True=LinearInterpolate(A=@n101.RGB, B=@n102, Alpha=@n103), False=@n101.RGB) {group=Nitro_Liquid, parameterName=UseBlending}) {name=LowEndNormalTexture, nodeColor=<rgba 0 0.8125 1 1>}
  NamedRerouteDeclaration(None=ScalarParameter() {defaultValue=-0.2, group=NitroLiquid-2DRayMarch, parameterName=NL_LowEndBumpOffset_Ratio}) {name=BumpOffsetRatio, nodeColor=<rgba 0.9219 1 0 1>}
  NamedRerouteDeclaration(None=VectorParameter() {group=NitroLiquid-VolumeTexture, parameterName=VolumeTextureTilingDistort, defaultValue=<rgba 1 1 2 1>}.RGB) {name=VolumeTextureTiling_Distort, nodeColor=<rgba 1 0.0703 0 1>}
  NamedRerouteDeclaration(None=VectorParameter() {group=NitroLiquid-VolumeTexture, parameterName=PanSpeedDistort, defaultValue=<rgba 0 0 -0.1 1>}.RGB) {name=PanSpeedDistort, nodeColor=<rgba 0 0.7187 1 1>}
  n104 = TextureObjectParameter() {parameterName=VolumeTexture, samplerType=SAMPLERTYPE_LinearGrayscale, texture=/Game/Global/MaterialFunctions/TechArt/Nitro/T_VT_Noise}
  NamedRerouteDeclaration(None=@n104) {name=VolumeTexture, nodeColor=<rgba 0 1 0.2266 1>}
  NamedRerouteDeclaration(None=StaticSwitchParameter(True=NamedRerouteUsage(), False=NamedRerouteUsage()) {parameterName=UsePsuedovolumeTexture}) {name=RayMarchedHeightMap, nodeColor=<rgba 0 0.6719 1 1>}
  NamedRerouteDeclaration(None=VectorParameter() {group=NitroLiquid-VolumeTexture, parameterName=VolumeTextureTiling, defaultValue=<rgba 2 2 1.5 1>}.RGB) {name=VolumeTextureTiling, nodeColor=<rgba 0.6172 0 1 1>}
  n105 = ObjectPositionWS()
  n106 = ParticlePositionWS()
  NamedRerouteDeclaration(None=StaticSwitchParameter(True=@n105, False=@n106) {defaultValue=true, group=Nitro_Liquid, parameterName=LocalizedMask}) {name=Position, nodeColor=<rgba 0 0.5078 1 1>}
  n107 = PreSkinnedLocalBounds()
  NamedRerouteDeclaration(None=Multiply(A=@n107.Extents, B=NamedRerouteUsage())) {name=Dimensions2, nodeColor=<rgba 0 0.0994 1 1>}
  NamedRerouteDeclaration(None=Transform(None=@n107.Min) {transformSourceType=TRANSFORMSOURCE_Instance}) {name=MinBounds, nodeColor=<rgba 0 0.4609 1 1>}
  n108 = ScreenPosition()
  n109 = MaterialFunctionCall(Float2=ViewSize()) {materialFunction=&BreakOutFloat2Components}
  n110 = AppendVector(A=Constant() {r=1}, B=Divide(A=@n109.G, B=@n109.G))
  n111 = ConstantBiasScale(None=Multiply(A=ConstantBiasScale(None=@n108.ViewportUV) {bias=-0.5, scale=2}, B=@n110))
  n112 = NamedRerouteUsage()
  n113 = NamedRerouteUsage()
  n114 = Multiply(A=NamedRerouteUsage(), B=NamedRerouteUsage())
  n115 = ScalarParameter() {defaultValue=0.5, parameterName=Screen Space UV Wobble}
  n116 = StaticSwitchParameter(True=ObjectPositionWS(), False=Subtract(A=Add(A=@n112, B=@n113), B=Multiply(A=@n114, B=@n115))) {parameterName=UseObjectPositionforBubbles}
  n117 = Multiply(
   A=@n110
   B=Multiply(
    A=Constant2Vector() {g=-0.5, r=0.5}
    B=MaterialFunctionCall(World Position=@n116) {materialFunction=&TransformToClipSpace}.Clip Space XY
   )
  )
  n118 = Multiply(A=NamedRerouteUsage(), B=ScalarParameter() {defaultValue=8, parameterName=Screen Space Scale})
  n119 = ComponentMask(None=TransformPosition(None=@n116) {transformSourceType=TRANSFORMPOSSOURCE_World, transformType=TRANSFORMPOSSOURCE_View}) {b=true}
  n120 = NamedRerouteDeclaration(None=MaterialFunctionCall(UVs=Subtract(A=@n111, B=@n117), Texture Scale=Divide(A=@n118, B=@n119) {constA=100}) {materialFunction=&ScaleUVsByCenter}.UVs) {name=ScreenSpaceUV, nodeColor=<rgba 1 0 0.8672 1>}
  n121 = MaterialFunctionCall(Float2=NamedRerouteUsage()) {materialFunction=&BreakOutFloat2Components}
  n122 = MaterialFunctionCall(X=@n121.R, Y=OneMinus(None=@n121.G)) {materialFunction=&MakeFloat2}
  NamedRerouteDeclaration(None=StaticSwitchParameter(True=@n120, False=StaticSwitchParameter(True=@n122.Result, False=@n120) {defaultValue=true, group=Nitro_Liquid, parameterName=LocalizedMask}) {parameterName=UseObjectPositionforBubbles}) {name=BubbleUV, nodeColor=<rgba 1 0 0.375 1>}
  NamedRerouteDeclaration(None=@n83.Time) {name=Niagara Time, nodeColor=<rgba 0 1 0.6719 1>}
  n123 = PreSkinnedPosition()
  n124 = ShaderStageSwitch(PixelShader=VertexInterpolator(VS=@n123).PS, VertexShader=@n123)
  n125 = NamedRerouteUsage()
  n126 = NamedRerouteUsage()
  n127 = NamedRerouteUsage()
  n128 = Saturate(None=Subtract(A=Subtract(A=@n124, B=@n125), B=Multiply(A=@n126, B=@n127) {constB=0.5}))
  NamedRerouteDeclaration(None=OneMinus(None=ComponentMask(None=@n128) {b=true})) {name=FillLevelMask, nodeColor=<rgba 0.4062 1 0 1>}
  NamedRerouteDeclaration(None=VectorParameter() {group=NitroLiquid-VolumeTexture, parameterName=PanSpeed3, defaultValue=<rgba 0 0 -0.3 1>}.RGB) {name=PanSpeedVolumeTexture, nodeColor=<rgba 0 0.8594 1 1>}
  NamedRerouteDeclaration(None=ScalarParameter() {defaultValue=0.05, group=NitroLiquid-VolumeTexture, parameterName=StepSizeBias}) {name=StepSizeBias, nodeColor=<rgba 0 1 0.5781 1>}
  n129 = VectorParameter() {parameterName=FillLevel, defaultValue=<rgba 0 0 0.5 1>}
  NamedRerouteDeclaration(None=Add(A=@n129.RGB, B=NamedRerouteUsage())) {name=FillLevel, nodeColor=<rgba 0 1 0.7422 1>}
  n130 = NamedRerouteUsage()
  NamedRerouteDeclaration(None=StaticSwitchParameter(True=ScalarParameter() {group=Nitro_Liquid, parameterName=Volatile}, False=Clamp(None=@n130) {maxDefault=0.5}) {defaultValue=true, group=Nitro_Liquid, parameterName=LocalizedMask}) {name=Volatile, nodeColor=<rgba 1 0 0.1641 1>}
  NamedRerouteDeclaration(None=ScalarParameter() {defaultValue=0.35, group=NitroLiquid-VolumeTexture, parameterName=distortstrength}) {name=distortstrength, nodeColor=<rgba 0.5234 0 1 1>}
  NamedRerouteDeclaration(None=ScalarParameter() {defaultValue=16, group=NitroLiquid-VolumeTexture, parameterName=XFrames}) {name=XFrames, nodeColor=<rgba 0 0.6953 1 1>}
  NamedRerouteDeclaration(None=Max(A=Max(A=@n82.Scale X, B=@n82.Scale X), B=@n82.Scale Z)) {name=Liquid Scale, nodeColor=<rgba 0.8047 0 1 1>}
  NamedRerouteDeclaration(None=Transform(None=@n107.Max) {transformSourceType=TRANSFORMSOURCE_Instance}) {name=MaxBounds, nodeColor=<rgba 0.3125 1 0 1>}
  n131 = Transform(None=ComponentMask(None=VectorParameter() {group=Static Overrides, parameterName=Center Offset}.RGB) {b=true, g=true, r=true}) {transformSourceType=TRANSFORMSOURCE_Local}
  n132 = Transform(None=Multiply(A=Add(A=@n81.Min, B=@n81.Min)) {constB=0.5}) {transformSourceType=TRANSFORMSOURCE_Instance}
  NamedRerouteDeclaration(None=StaticSwitchParameter(True=@n131, False=@n132) {defaultValue=true, group=Nitro_Liquid, parameterName=LocalizedMask}) {name=Center Offset, nodeColor=<rgba 1 0 0.7266 1>}
  NamedRerouteDeclaration(None=MaterialFunctionCall() {materialFunction=&ObjectScale}.Scale XYZ) {name=ObjectScale, nodeColor=<rgba 1 0 0.3516 1>}
  NamedRerouteDeclaration(None=ComponentMask(None=@n31) {r=true}) {name=Radius, nodeColor=<rgba 1 0.4453 0 1>}
  n133 = TransformPosition(None=Constant3Vector()) {transformType=TRANSFORMPOSSOURCE_World}
  n134 = Constant()
  n135 = MaterialFunctionCall(X=@n134, Y=@n134, Z=ComponentMask(None=NamedRerouteUsage()) {b=true}) {materialFunction=&MakeFloat3}
  n136 = NamedRerouteUsage()
  n137 = NamedRerouteUsage()
  n138 = MaterialFunctionCall(
   Plane Normal Axis=Transform(None=Constant3Vector() {constant=<rgba 0 0 1 0>}) {transformSourceType=TRANSFORMSOURCE_Local}
   Plane X Axis=Transform(None=Constant3Vector() {constant=<rgba 1 0 0 0>}) {transformSourceType=TRANSFORMSOURCE_Local}
   Plane Center=Add(A=Add(A=@n133, B=@n135.Result), B=Multiply(A=@n136, B=@n137) {constB=0.5})
  ) {materialFunction=&VirtualPlaneCoordinates}
  n139 = Abs(None=ComponentMask(None=@n138.UVs) {r=true})
  n140 = Abs(None=ComponentMask(None=@n138.UVs) {g=true})
  n141 = Max(A=Round(None=Saturate(None=@n139)), B=Round(None=Saturate(None=@n140)))
  n142 = SphereMask(A=@n138.UVs, B=Constant(), Radius=NamedRerouteUsage(), Hardness=Constant() {r=1})
  n143 = NamedRerouteDeclaration(None=StaticSwitchParameter(True=Multiply(A=@n141, B=@n142), False=NamedRerouteUsage()) {group=Nitro_Liquid, parameterName=LocalizedMask}) {name=Surfacemask, nodeColor=<rgba 0.5703 1 0 1>}
  n144 = NamedRerouteUsage()
  NamedRerouteDeclaration(None=StaticSwitchParameter(True=Max(A=@n143, B=@n144), False=@n84.FluidMask) {defaultValue=true, group=Nitro_Liquid, parameterName=LocalizedMask}) {name=FinalFillMask, nodeColor=<rgba 0.1016 1 0 1>}
  NamedRerouteDeclaration(None=ScalarParameter() {defaultValue=16, group=NitroLiquid-VolumeTexture, parameterName=YFrames}) {name=YFrames, nodeColor=<rgba 0 0.5547 1 1>}
  NamedRerouteDeclaration(None=ComponentMask(None=VectorParameter() {group=NitroLiquid-2DRayMarch, parameterName=NL_DistortionPanSpeed, defaultValue=<rgba 0 -0.2 0 1>}.RGB) {g=true, r=true}) {name=DistortionPanSpeed, nodeColor=<rgba 0 1 0.1094 1>}
  n145 = Constant()
  n146 = Multiply(
   A=LinearInterpolate(
    A=ScalarParameter() {defaultValue=-0.1, parameterName=SurfaceDistortionStrength}
    B=ScalarParameter() {defaultValue=-0.5, parameterName=SurfaceDistortionStrength2}
    Alpha=NamedRerouteUsage()
   ) {constA=14, constB=7}
   B=@n60
  )
  NamedRerouteDeclaration(None=StaticSwitchParameter(
    True=MaterialFunctionCall(X=@n145, Y=@n145, Z=@n146) {materialFunction=&MakeFloat3}.Result
    False=@n146
   ) {group=Nitro_Liquid, parameterName=LocalizedMask}) {name=SurfaceDistortion, nodeColor=<rgba 1 0.5859 0 1>}
  NamedRerouteDeclaration(None=@n75.Param3) {name=Foam Time, nodeColor=<rgba 0 1 0.7656 1>}
  NamedRerouteDeclaration(None=@n129.RGB) {name=FillLevel_NoDistortion, nodeColor=<rgba 0.3359 1 0 1>}
  n147 = NamedRerouteUsage()
  n148 = Multiply(A=@n147, B=NamedRerouteUsage())
  n149 = NamedRerouteUsage()
  n150 = NamedRerouteUsage()
  n151 = Add(A=Multiply(A=Time(), B=NamedRerouteUsage()), B=AppendVector(A=Constant2Vector(), B=Multiply(A=@n149, B=@n150) {constB=4}))
  NamedRerouteDeclaration(None=Frac(None=Add(A=@n148, B=@n151))) {name=CurPos, nodeColor=<rgba 0 0.3906 1 1>}
  n152 = Multiply(A=@n147, B=NamedRerouteUsage())
  n153 = NamedRerouteUsage()
  n154 = NamedRerouteUsage()
  n155 = Add(A=Multiply(A=Time(), B=NamedRerouteUsage()), B=AppendVector(A=Constant2Vector(), B=Multiply(A=@n153, B=@n154) {constB=4}))
  NamedRerouteDeclaration(None=Frac(None=Add(A=@n152, B=@n155))) {name=CurPosDistort, nodeColor=<rgba 0.3594 0 1 1>}
  NamedRerouteDeclaration(None=Transform(None=CameraVectorWS()) {transformSourceType=TRANSFORMSOURCE_World, transformType=TRANSFORM_Instance}) {name=viewDir, nodeColor=<rgba 0.0312 0 1 1>}
  MaterialExpressionNamedRerouteUsage_65 = NamedRerouteUsage()
  MaterialExpressionAppendVector_1 = AppendVector(A=NamedRerouteUsage(), B=NamedRerouteUsage())
  MaterialExpressionMultiply_30 = Multiply(A=NamedRerouteUsage(), B=NamedRerouteUsage())
  MaterialExpressionNamedRerouteUsage_87 = NamedRerouteUsage()
  MaterialExpressionNamedRerouteUsage_79 = NamedRerouteUsage()
  MaterialExpressionNamedRerouteUsage_92 = NamedRerouteUsage()
  MaterialExpressionNamedRerouteUsage_97 = NamedRerouteUsage()
  n156 = Custom!(
   Tex=@MaterialExpressionNamedRerouteUsage_65
   XYFrames=@MaterialExpressionAppendVector_1
   NumFrames=@MaterialExpressionMultiply_30
   LocalCamVec=@MaterialExpressionNamedRerouteUsage_87
   CurPos=@MaterialExpressionNamedRerouteUsage_79
   CurPosDistort=@MaterialExpressionNamedRerouteUsage_92
   distortstrength=@MaterialExpressionNamedRerouteUsage_97
  ) {
   code="// The below expression will get compiled\r\n// into the output of this node\r\nfloat accumdens = 0;\r\nfloat offset =0.025;\r\n//float cursample = 0;\r\nfloat3 cursampledistort = PseudoVolumeTexture(Tex,TexSampler,CurPosDistort,XYFrames,NumFrames)*distortstrength;\r\nfloat cursample = PseudoVolumeTexture(Tex,TexSampler,CurPos+cursampledistort,XYFrames,NumFrames).r;\r\n\r\nfloat OffsetR = PseudoVolumeTexture(Tex,TexSampler,CurPos+cursampledistort+float3(offset,0,0),XYFrames,NumFrames).r;\r\nfloat OffsetG = PseudoVolumeTexture(Tex,TexSampler,CurPos+cursampledistort+float3(0,offset,0),XYFrames,NumFrames).r;\r\n\r\nNr = cursample - OffsetR;\r\nNg = cursample - OffsetG;\r\n\r\nreturn cursample;"
   description=LowEndVolumeSample
   showCode=true
   additionalOutputs [{outputName=Nr, outputType=CMOT_Float1}, {outputName=Ng, outputType=CMOT_Float1}]
   inputs [
    {inputName=Tex, input {expression=&self:MaterialExpressionNamedRerouteUsage_65}}
    {inputName=XYFrames, input {expression=&self:MaterialExpressionAppendVector_1}}
    {inputName=NumFrames, input {expression=&self:MaterialExpressionMultiply_30}}
    {
     inputName=LocalCamVec
     input {expression=&self:MaterialExpressionNamedRerouteUsage_87}
    }
    {
     inputName=CurPos
     input {expression=&self:MaterialExpressionNamedRerouteUsage_79}
    }
    {
     inputName=CurPosDistort
     input {expression=&self:MaterialExpressionNamedRerouteUsage_92}
    }
    {
     inputName=distortstrength
     input {expression=&self:MaterialExpressionNamedRerouteUsage_97}
    }
   ]
  }
  NamedRerouteDeclaration(None=Saturate(None=Power(Base=@n156.return) {constExponent=3.5})) {name=LowEndVolumeTexture, nodeColor=<rgba 0 0.8828 1 1>}
  n157 = ScalarParameter() {defaultValue=5, parameterName=Normal_Strength}
  n158 = Multiply(A=@n156.Nr, B=@n157)
  n159 = Multiply(A=@n156.Ng, B=@n157)
  n160 = Constant() {r=1}
  NamedRerouteDeclaration(None=Normalize(VectorInput=MaterialFunctionCall(X=@n158, Y=@n159, Z=@n160) {materialFunction=&MakeFloat3}.Result)) {name=LowEndVolumeNormal, nodeColor=<rgba 0 1 0.8125 1>}
  NamedRerouteDeclaration(None=ScalarParameter() {defaultValue=2.5, group=NitroLiquid-VolumeTexture, parameterName=VolatilePanSpeedBias}) {name=VolatilePanSpeedBias, nodeColor=<rgba 0.0312 0 1 1>}
  n161 = TextureObject() {
   samplerType=SAMPLERTYPE_LinearGrayscale
   texture=&T_Swirly_Noise_MiddleDetail
  }
  NamedRerouteDeclaration(None=@n161) {name=DistortionTexture, nodeColor=<rgba 0 1 0.2969 1>}
  n162 = Constant()
  NamedRerouteDeclaration(None=Multiply(A=AppendVector(A=@n162, B=@n2)) {constB=-1}) {name=BubbleTime, nodeColor=<rgba 1 0.4688 0 1>}
  NamedRerouteDeclaration(None=MaterialFunctionCall(
    X=ScalarParameter() {group=NitroLiquid-2DRayMarch, parameterName=NoiseDistortionPanSpeed_X}
    Y=ScalarParameter() {defaultValue=-0.2, group=NitroLiquid-2DRayMarch, parameterName=NoiseDistortionPanSpeed_Y}
   ) {materialFunction=&MakeFloat2}.Result) {name=NoiseDistortionPanSpeed, nodeColor=<rgba 0.7578 0 1 1>}
  NamedRerouteDeclaration(None=MaterialFunctionCall(
    X=ScalarParameter() {group=NitroLiquid-2DRayMarch, parameterName=NoisePanSpeed_X}
    Y=ScalarParameter() {defaultValue=-0.1, group=NitroLiquid-2DRayMarch, parameterName=NoisePanSpeed_Y}
   ) {materialFunction=&MakeFloat2}.Result) {name=NoisePanSpeed, nodeColor=<rgba 0 1 0.4375 1>}
  n163 = MaterialFunctionCall(Float2=TextureCoordinate()) {materialFunction=&BreakOutFloat2Components}
  n164 = ConstantBiasScale(None=Sine(None=Add(A=@n163.G) {constB=0.25}))
  n165 = Constant() {r=1}
  NamedRerouteDeclaration(None=Max(A=NamedRerouteUsage(), B=SphereMask(A=@n164, B=@n165) {attenuationRadius=0.05, hardnessPercent=0})) {name=BlendMsk, nodeColor=<rgba 1 0 0.6797 1>}
  NamedRerouteDeclaration(None=MaterialFunctionCall(
    X=ScalarParameter() {defaultValue=1, group=NitroLiquid-2DRayMarch, parameterName=NoiseTiling_X}
    Y=ScalarParameter() {defaultValue=1, group=NitroLiquid-2DRayMarch, parameterName=NoiseTiling_Y}
   ) {materialFunction=&MakeFloat2}.Result) {name=NoiseTiling, nodeColor=<rgba 0.7109 0 1 1>}
  NamedRerouteDeclaration(None=Multiply(A=TextureCoordinate(), B=NamedRerouteUsage())) {name=UV, nodeColor=<rgba 0 1 0.0391 1>}
  n166 = Constant() {r=1}
  n167 = MaterialFunctionCall(Normal=NamedRerouteUsage(), Flatness=Constant() {r=0.5}) {materialFunction=&FlattenNormal}
  n168 = NamedRerouteUsage()
  n169 = LinearInterpolate(A=NamedRerouteUsage(), B=NamedRerouteUsage(), Alpha=NamedRerouteUsage())
  n170 = SetMaterialAttributes!(
   MaterialAttributes=NamedRerouteUsage()
   BaseColor=@n169
   Specular=ScalarParameter() {defaultValue=1, parameterName=NitroSpec}
   Roughness=ScalarParameter() {defaultValue=0.1, parameterName=NitroRoughness}
   EmissiveColor=@n169
   Normal=LinearInterpolate(
    A=@n168
    B=MaterialFunctionCall(BaseNormal=@n168, AdditionalNormal=@n167.Result) {materialFunction=&BlendAngleCorrectedNormals}.Result
    Alpha=NamedRerouteUsage()
   )
   OpacityMask=@n166
   Opacity=@n166
  ) {attributeSetTypes [
    #3
    #4
    #5
    #6
    #7
    #1
    #2
   ]}
  NamedRerouteDeclaration(None=@n170) {name=Nitro_Material, nodeColor=<rgba 0 0.7656 1 1>}
  NamedRerouteUsage()
  NamedRerouteUsage()
  NamedRerouteUsage()
  NamedRerouteUsage()
  NamedRerouteUsage()
  NamedRerouteUsage()
  NamedRerouteUsage()
  NamedRerouteUsage()
  NamedRerouteUsage()
  NamedRerouteUsage()
  NamedRerouteUsage()
  NamedRerouteUsage()
  NamedRerouteUsage()
  NamedRerouteUsage()
  n171 = ScalarParameter() {defaultValue=5, group=NitroLiquid-2DRayMarch, parameterName=Normal_Strength}
  Normalize(VectorInput=MaterialFunctionCall(X=Multiply(A=@n67.Nr, B=@n171), Y=Multiply(A=@n67.Ng, B=@n171), Z=@n67.Nb) {materialFunction=&MakeFloat3}.Result)
  OneMinus(None=@n93.RGB)
  OneMinus(None=NamedRerouteUsage())
  ParticlePositionWS()
  Power(Base=@n51)
  ScalarParameter() {defaultValue=2, parameterName=BubbleVolatilitySpeed}
  ScalarParameter() {defaultValue=100, parameterName=Scale}
  StaticSwitchParameter(True=@n106, False=@n105) {defaultValue=true, parameterName=Is Niagara}
  StaticSwitchParameter(True=@n132, False=@n131) {defaultValue=true, parameterName=Is Niagara}
  Subtract(A=@n1, B=@n2)
  Subtract(A=@MaterialExpressionTime_4, B=@MaterialExpressionMultiply_46)
  TextureCoordinate()
  TextureCoordinate()
  TextureCoordinate()
  TextureCoordinate()
  TextureSample() {samplerType=SAMPLERTYPE_Grayscale, texture=/Game/Effects/Fort_Effects/Textures/Fire/T_WispyFireNoise_Soft_M}
  TextureSample() {texture=/Game/Global/Textures/Default/Tileable_Patterns/T_Thick_Smoke_05}
  TextureSample() {
   samplerType=SAMPLERTYPE_LinearGrayscale
   texture=&T_Swirly_Noise_MiddleDetail
  }
  TextureSample() {texture=&T_SoftNoise}
  Transform(None=@n63) {transformSourceType=TRANSFORMSOURCE_Camera, transformType=TRANSFORM_Tangent}
 ]}
